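# Niagara effect for the BriskReign Glacier glider: timed twin beams with flickering glows and a light, energy meshes, ice shards and mist.
# Five emitters use PartitionParticles to index skeletal mesh bones or sockets; beams, glows and energy follow them, while shards and mist spray downward.
# Staggered loop delays start glows and energy before beams, shards and mist.
# User.Variant switches cyan to orange (light stays cyan); time-of-day, front-end and platform modifiers tune brightness.
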NiagaraSystem /BRCosmetics/Gadgets/Assets/VinderTech_GliderChute/Glider_BriskReign_Glacier/FX/NS_Glider_BriskReign_Beams
dependencies {
 usedDataInterfaces [
  &NiagaraDataInterfaceCurve
  &NiagaraDataInterfaceArrayColor
  &NiagaraDataInterfaceArrayFloat3
  &NiagaraDataInterfacePlatformSet
  &NiagaraDataInterfaceVectorCurve
  &NiagaraDataInterfaceColorCurve
  &NiagaraDataInterfaceSpriteRendererInfo
  &NiagaraDataInterfaceArrayFloat
  &NiagaraDataInterfaceSkeletalMesh
 ]
 usedDynamicInputs [&FN_FrontEndVector3Modifier, &Vector2DFromFloat]
 usedModules [
  &self:SystemSpawnScript.SystemScriptSource.SystemScriptGraph.NiagaraNodeAssignment_0.SetVariables_#1
  &SystemState
  &self:SystemSpawnScript.SystemScriptSource.SystemScriptGraph.NiagaraNodeAssignment_2.SetVariables_#2
 ]
}
emitters {
 Beams {
  emitterUpdateScript [
   &EmitterState(
    Inactive Response=::NewEnumerator1
    Life Cycle Mode=::NewEnumerator1
    Loop Behavior=::NewEnumerator1
    Loop Delay=0.6
    Loop Duration=2
    UseLoopDelay=true
   )
   &SpawnBurst_Instantaneous(Spawn Count=2, Use Spawn Probability=false, UseLoopCountLimit=false)
   /SetVariables(Emitter.SplineWidth_Base:NiagaraFloat=&FloatFromCurve(
     FloatCurve=DataInterface() {propertyValues {Curve {keys [
         <at 0 2>
         <at 0.1 1>
         {arriveTangent=-6.7804, interpMode=RCIM_Cubic, leaveTangent=-6.7804, tangentMode=RCTM_User, time=0.7, value=1}
         {arriveTangent=-0.8494, interpMode=RCIM_Cubic, leaveTangent=-0.8494, tangentMode=RCTM_User, time=1, value=0}
        ]}}}
     CurveIndex=$Emitter.NormalizedLoopAge
     Scale Curve=20
    )) {isSetParameters=true}
   /SetVariables(Emitter.SplineWidth_Tip:NiagaraFloat=&FloatFromCurve(
     FloatCurve=DataInterface() {propertyValues {Curve {keys [
         <at 0 1>
         {arriveTangent=-6.7804, interpMode=RCIM_Cubic, leaveTangent=-6.7804, tangentMode=RCTM_User, time=0.7, value=1}
         {arriveTangent=-0.8494, interpMode=RCIM_Cubic, leaveTangent=-0.8494, tangentMode=RCTM_User, time=1, value=0}
        ]}}}
     CurveIndex=$Emitter.NormalizedLoopAge
     Scale Curve=20
    )) {isSetParameters=true}
   /SetVariables(Emitter.SplineWidthRed_Base:NiagaraFloat=&FloatFromCurve(
     FloatCurve=DataInterface() {propertyValues {Curve {keys [
         <at 0 2>
         <at 0.1 1>
         {arriveTangent=-4.6776, interpMode=RCIM_Cubic, leaveTangent=-4.6776, tangentMode=RCTM_User, time=0.6, value=1}
         {arriveTangent=-0.5343, interpMode=RCIM_Cubic, leaveTangent=-0.5343, tangentMode=RCTM_User, time=1, value=0}
        ]}}}
     CurveIndex=$Emitter.NormalizedLoopAge
     Scale Curve=15
    )) {isSetParameters=true}
   /SetVariables(Emitter.SplineWidthRed_Tip:NiagaraFloat=&FloatFromCurve(
     FloatCurve=DataInterface() {propertyValues {Curve {keys [
         <at 0 1>
         {arriveTangent=-4.6776, interpMode=RCIM_Cubic, leaveTangent=-4.6776, tangentMode=RCTM_User, time=0.6, value=1}
         {arriveTangent=-0.5343, interpMode=RCIM_Cubic, leaveTangent=-0.5343, tangentMode=RCTM_User, time=1, value=0}
        ]}}}
     CurveIndex=$Emitter.NormalizedLoopAge
     Scale Curve=15
    )) {isSetParameters=true}
  ]
  particleSpawnScript [
   &PartitionParticles()
   &InitializeParticle(
    Color=&SelectLinearColorFromArray(
     Array Sampling Mode=::NewEnumerator1
     Direct Array Index=&MakeIntFromFloat(Float=$User.Variant)
     Color Selection Array=DataInterface() {propertyValues {ColorData [<rgba 0 0.01 0.01 1.5>, <rgba 0.01 0.0058 0 1.5>]}}
     Random Seed=?
    )
    Lifetime=0.92
    Mesh Scale=<1 1 3.9>
    Mesh Scale Mode=::NewEnumerator4
    UsePositionOffset=false
    Write Material Random=true
   )
   &InitialMeshOrientation(Mesh Orientation Mode=::NewEnumerator2, Use Rotation Vector=true)
  ]
  particleUpdateScript [
   &ParticleState()
   &SkeletalMeshLocation(
    Bone / Socket Index=$Particles.PartitionParticles.Partition
    Bone Sampling Mode=::NewEnumerator4
    MaskSpawnGroup=false
    Orientation Sampling=::NewEnumerator1
    Skeletal Mesh=$System.Skeletal Mesh Reader
   )
   &UpdateMeshOrientation(
    Facing Direction=&SelectVectorFromArray(
     Direct Array Index=$Particles.PartitionParticles.Partition
     Vector Selection Array=DataInterface() {propertyValues {FloatData [<0 -0.15 -0.05>, <0 0.15 0.05>]}}
    )
    Facing Interpolation=false
    Facing Interpolation Rate=5
    Orientation Method=::NewEnumerator1
    Override Quat=false
    TransformFacing=false
    TransformUp=false
    Up Direction:Vector3f=<0 0 -1>
   )
   &ScaleMeshSize(Scale Factor=&Multiply_Vector(
     A=&MakeVector(
      X=1
      Y=1
      Z=&FloatFromCurve(FloatCurve=DataInterface() {propertyValues {Curve {keys [<at 0 0>,<at 0.1 1>,<at 1 1.2>]}}})
     )
     B=&Divide_Vector(A:Vector3f=<1 1 1>, B:Vector3f=$Engine.Owner.Scale)
    ))
   &ScaleColor(
    Scale Alpha=&Multiply_Float(
     A=&FloatFromCurve(
      FloatCurve=DataInterface() {propertyValues {Curve {keys [
          <at 0 1>
          {arriveTangent=-0.2775, interpMode=RCIM_Cubic, leaveTangent=-0.2775, tangentMode=RCTM_User, time=0.8, value=1}
          {arriveTangent=-2.6012, interpMode=RCIM_Cubic, leaveTangent=-2.6012, tangentMode=RCTM_User, time=0.9, value=0.85}
          {arriveTangent=-14.0132, interpMode=RCIM_Cubic, leaveTangent=-14.0132, tangentMode=RCTM_User, time=1, value=0}
         ]}}}
      Scale Curve=&FNTimeOfDayScalar(Night Scale=0.6, Morning Scale=0.8, Evening Scale=0.6)
     )
     B=&FN_FrontEndFloatModifier(Gameplay=1, Front End=0.7)
    )
    ScaleA=true
    ScaleRGB=true
   )
   /SetVariables(Particles.color002:LinearColor=&MakeLinearColorFromVectorAndFloat(
     "Vector (RGB)"=&MakeVectorFromLinearColorRGB(Color=<rgba 1 1 1 1>)
     "Float (Alpha)"=&Multiply_Float(
      A=&FloatFromCurve(
       FloatCurve=DataInterface() {propertyValues {Curve {keys [<at 0 1>,<at 0.95 1>,<at 1 0>]}}}
       Scale Curve=&FNTimeOfDayScalar(Night Scale=0.2, Morning Scale=0.2, Day Scale=0.3, Evening Scale=0.2)
      )
      B=&Multiply_Float(
       A=&FN_FrontEndFloatModifier(Gameplay=1, Front End=0.8)
       B=&FN_PickFloatDependOnPlatform(TargetPlatformActive=1, TargetPlatformNoActive=0.8)
      )
     )
    )) {isSetParameters=true}
   /SetVariables(Particles.Color003:LinearColor=&MakeLinearColorFromVectorAndFloat(
     "Vector (RGB)"=&MakeVectorFromLinearColorRGB(Color=<rgba 1.5 1.5 1.5 1>)
     "Float (Alpha)"=&Multiply_Float(
      A=&FloatFromCurve(
       FloatCurve=DataInterface() {propertyValues {Curve {keys [<at 0 5>,<at 0.1 1>,<at 0.9 1>,<at 1 0>]}}}
       Scale Curve=&FNTimeOfDayScalar(Night Scale=0.6, Morning Scale=0.8, Evening Scale=0.6)
      )
      B=&FN_PickFloatDependOnPlatform(TargetPlatformActive=1, TargetPlatformNoActive=0.15)
     )
    )) {isSetParameters=true}
   /SetVariables(Particles.Color004:LinearColor=&MakeLinearColorFromVectorAndFloat(
     "Vector (RGB)"=&MakeVectorFromLinearColorRGB(Color=<rgba 1 1 1 1.5>)
     "Float (Alpha)"=&Multiply_Float(
      A=&FloatFromCurve(
       FloatCurve=DataInterface() {propertyValues {Curve {keys [<at 0 1>,<at 0.85 1>,<at 1 0>]}}}
       Scale Curve=&FNTimeOfDayScalar(Night Scale=0.6, Morning Scale=0.8, Evening Scale=0.6)
      )
      B=&FN_PickFloatDependOnPlatform(TargetPlatformActive=1, TargetPlatformNoActive=0.6)
     )
    )) {isSetParameters=true}
   &DynamicMaterialParameters(
    Index 0 Param 1=&FloatFromCurve(
     FloatCurve=DataInterface() {propertyValues {Curve {keys [
         <at 0 1 RCIM_Cubic>
         {arriveTangent=-29.3222, leaveTangent=-2.8552, tangentMode=RCTM_User, time=0.1, value=-1}
         {arriveTangent=12.0832, interpMode=RCIM_Cubic, leaveTangent=12.0832, tangentMode=RCTM_User, time=0.7005, value=-1}
         {arriveTangent=0.7491, interpMode=RCIM_Cubic, leaveTangent=0.7491, tangentMode=RCTM_User, time=1, value=1}
        ]}}}
     Scale Curve=&RandomRangeFloat(Minimum=0.9, Maximum=1)
    )
    Index 0 Param 2=1
    Index 0 Param 3=1
    Module.Index0Param0Write=true
    Module.Index0Param1Write=true
    Module.Index0Param2Write=true
    Module.Index0Param3Write=true
   )
  ]
  rendererClasses [&NiagaraMeshRendererProperties]
  renderers [
   {rendererClass=&NiagaraMeshRendererProperties, rendererIndex=0}
   {rendererClass=&NiagaraMeshRendererProperties, rendererIndex=1}
   {rendererClass=&NiagaraMeshRendererProperties, rendererIndex=2}
   {rendererClass=&NiagaraMeshRendererProperties, rendererIndex=3}
  ]
 }
 Glows {
  emitterUpdateScript [
   &EmitterState(
    Inactive Response=::NewEnumerator1
    Life Cycle Mode=::NewEnumerator1
    Loop Behavior=::NewEnumerator1
    Loop Delay=0.2
    Loop Duration=2.4
    UseLoopDelay=true
   )
   &SpawnBurst_Instantaneous(Spawn Count=2, Use Spawn Probability=false, UseLoopCountLimit=false)
  ]
  particleSpawnScript [
   &PartitionParticles()
   &InitializeParticle(
    Color=<rgba 0.15 0.15 0.15 1.5>
    Lifetime=1.32
    Mesh Scale=<0.05 0.08 0.05>
    Mesh Scale Mode=::NewEnumerator4
    Sprite Size=&Multiply_Vector2D_ByFloat(Float=&FN_FrontEndFloatModifier(Gameplay=0.75, Front End=1), Vector2D=<160 60>)
    Sprite Size Mode=::NewEnumerator4
    UsePositionOffset=false
    Write Material Random=true
   )
   &InitialMeshOrientation(Mesh Orientation Mode=::NewEnumerator2, Orientation Vector=<0 0 1>, Rotation=<0 0.25 0>, Use Rotation Vector=true)
   &CameraOffset(Camera Offset Amount=5)
   &AddVelocity(Velocity=<0 0 -0.001>)
  ]
  particleUpdateScript [
   &ParticleState()
   &SkeletalMeshLocation(
    Bone / Socket Index=$Particles.PartitionParticles.Partition
    Bone Sampling Mode=::NewEnumerator4
    MaskSpawnGroup=false
    Orientation Sampling=::NewEnumerator1
    Skeletal Mesh=$System.Skeletal Mesh Reader
   )
   &ScaleMeshSize(Scale Factor=&Multiply_Vector(
     A=&ScaleFloatByCurve(
      InitialValue:Vector3f=&Multiply_VectorByFloat(Vector=<1 1 1>, Float=&FN_FrontEndFloatModifier(Gameplay=0.75, Front End=1))
      ScaleCurve:NiagaraDataInterfaceVectorCurve=DataInterface() {propertyValues {
        XCurve {keys [<at 0 0.25>,<at 0.1 1>,<at 0.9 1>,<at 1 0>]}
        YCurve {keys [<at 0 1.25>,<at 0.1 2.25>,<at 0.9 2.25>,<at 1 0>]}
        ZCurve {keys [<at 0 1.25>,<at 0.1 2.25>,<at 0.9 2.25>,<at 1 0>]}
       }}
      CurveIndex:NiagaraFloat=$Particles.NormalizedAge
     )
     B=&Divide_Vector(A:Vector3f=<1 1 1>, B:Vector3f=$Engine.Owner.Scale)
    ))
   &ScaleSpriteSize(Uniform Curve Sprite Scale=DataInterface() {propertyValues {Curve {keys [
        <at 0 0>
        <at 0.1 1>
        <at 0.1512 0.7174>
        <at 0.2132 1>
        <at 0.2641 0.7539>
        <at 0.299 1>
        <at 0.3563 0.7037>
        <at 0.4151 1>
        <at 0.4564 0.8017>
        <at 0.5041 1>
        <at 0.5502 0.7561>
        <at 0.5836 1>
        <at 0.6551 0.7653>
        <at 0.6981 1>
        <at 0.7521 0.7721>
        <at 0.7871 1>
        <at 0.8475 0.7106>
        <at 0.9 1>
        {tangentMode=RCTM_User, time=1, value=0}
       ]}}})
   /SetVariables(Particles.Size002:Vector2f=&Multiply_Vector2D(A=$Particles.SpriteSize, B=<0.7 1>)) {isSetParameters=true}
   /SetVariables(Particles.Size003:Vector2f=&Multiply_Vector2D(A=<112.5 37.5>, B=<1 3>)) {isSetParameters=true}
   &ScaleColor(
    Scale Alpha=&Multiply_Float(
     A=&FloatFromCurve(
      FloatCurve=DataInterface() {propertyValues {Curve {keys [
          {arriveTangent=4.4475, time=0, value=0}
          {arriveTangent=5.3518, leaveTangent=-3.1774, time=0.1, value=1.2}
          {arriveTangent=-1.4972, leaveTangent=-1.4972, time=0.1523, value=0.7}
          {arriveTangent=-0.1667, leaveTangent=-0.1667, time=0.2052, value=1.1711}
          <at 0.258 0.7713>
          <at 0.2926 1.1775>
          {arriveTangent=-0.25, leaveTangent=-0.25, time=0.3608, value=0.689}
          <at 0.3865 1.193>
          <at 0.4263 0.8642>
          <at 0.4931 1.1191>
          {arriveTangent=-0.25, leaveTangent=-0.25, time=0.5428, value=0.7548}
          <at 0.5846 1.1063>
          <at 0.6571 0.8076>
          <at 0.6976 1.1685>
          {arriveTangent=0.1667, leaveTangent=0.1667, time=0.7542, value=0.7713}
          <at 0.7906 1.1138>
          <at 0.8535 0.8807>
          {arriveTangent=-3.3333, leaveTangent=-3.3333, time=0.9, value=1.2}
          <at 1 0>
         ]}}}
      Scale Curve=&FNTimeOfDayScalar(Night Scale=0.6, Morning Scale=0.7, Day Scale=0.8, Evening Scale=0.6)
     )
     B=&FN_FrontEndFloatModifier(Gameplay=1, Front End=0.3)
    )
    Scale RGB=&MakeVectorFromLinearColorRGB(Color=&MergeHelper_ColorSwitch(
      0=&ColorFromCurve(ColorCurve=DataInterface() {propertyValues {
         AlphaCurve {keys []}
         BlueCurve {keys [<at 0.7887 0.4375 RCIM_Cubic>,<at 0.9801 0.1562 RCIM_Cubic>]}
         GreenCurve {keys [<at 0.7887 0.4375 RCIM_Cubic>,<at 0.9801 0.1562 RCIM_Cubic>]}
         RedCurve {keys [<at 0.7887 0 RCIM_Cubic>,<at 0.9801 0 RCIM_Cubic>]}
        }})
      1=&ColorFromCurve(ColorCurve=DataInterface() {propertyValues {
         AlphaCurve {keys []}
         BlueCurve {keys [<at 0.7887 0 RCIM_Cubic>,<at 0.9801 0 RCIM_Cubic>]}
         GreenCurve {keys [<at 0.7887 0.2552 RCIM_Cubic>,<at 0.9801 0.0911 RCIM_Cubic>]}
         RedCurve {keys [<at 0.7887 0.4375 RCIM_Cubic>,<at 0.9801 0.1562 RCIM_Cubic>]}
        }})
      2=?
      3=?
      Index=&MakeIntFromFloat(Float=$User.Variant)
      Num=::NewEnumerator0
     ))
    ScaleA=true
    ScaleRGB=true
   )
   /SetVariables(Particles.color002:LinearColor=&MakeLinearColorFromVectorAndFloat(
     "Vector (RGB)"=&MergeHelper_VectorSwitch(
      0=&MakeVectorFromLinearColorRGB(Color=<rgba 0.12 2.04 3 1>)
      1=&MakeVectorFromLinearColorRGB(Color=<rgba 3 1.32 0.12 1>)
      2=?
      3=?
      Index=&MakeIntFromFloat(Float=$User.Variant)
      Num=::NewEnumerator0
     )
     "Float (Alpha)"=&FloatFromCurve(
      FloatCurve=DataInterface() {propertyValues {Curve {keys [<at 0 1>,<at 0.1 2>,<at 0.2005 1.7767>,<at 0.306 1.9927>,<at 0.3889 1.7548>,<at 0.5026 2.0076>,<at 0.5773 1.7477>,<at 0.6601 1.9791>,<at 0.7283 1.7134>,<at 0.8209 1.9848>,<at 1 0>]}}}
      Scale Curve=&FN_PickFloatDependOnPlatform(TargetPlatformActive=5, TargetPlatformNoActive=1)
     )
    )) {isSetParameters=true}
   /SetVariables(Particles.Color003:LinearColor=&MakeLinearColorFromVectorAndFloat(
     "Vector (RGB)"=&MergeHelper_VectorSwitch(
      0=&MakeVectorFromLinearColorRGB(Color=<rgba 0.04 1 1 1>)
      1=&MakeVectorFromLinearColorRGB(Color=<rgba 1 0.6 0.04 1>)
      2=?
      3=?
      Index=&MakeIntFromFloat(Float=$User.Variant)
      Num=::NewEnumerator0
     )
     "Float (Alpha)"=&Multiply_Float(
      A=&FloatFromCurve(
       FloatCurve=DataInterface() {propertyValues {Curve {keys [<at 0 0.8>,<at 0.1 1>,<at 0.2 0.6>,<at 0.3 1>,<at 0.4 0.7>,<at 0.5 0.95>,<at 0.6 0.55>,<at 0.7 1>,<at 0.8 0.55>,<at 0.9 0.8>,<at 1 0>]}}}
       Scale Curve=&FNTimeOfDayScalar(Night Scale=0.2, Morning Scale=0.1, Day Scale=0.4, Evening Scale=0.2)
      )
      B=&Multiply_Float(
       A=&FN_FrontEndFloatModifier(Gameplay=0.5, Front End=0.15)
       B=&FN_PickFloatDependOnPlatform(TargetPlatformActive=1, TargetPlatformNoActive=0.4)
      )
     )
    )) {isSetParameters=true}
   /SetVariables(Particles.MeshColor:LinearColor=&MakeLinearColorFromVectorAndFloat(
     "Vector (RGB)"=&MergeHelper_VectorSwitch(
      0=&MakeVectorFromLinearColorRGB(Color=&ColorFromCurve(ColorCurve=DataInterface() {propertyValues {
          AlphaCurve {keys []}
          BlueCurve {keys [
            {arriveTangent=-1.8762, interpMode=RCIM_Cubic, leaveTangent=-1.8762, tangentMode=RCTM_User, time=0.1298, value=2}
            {interpMode=RCIM_Cubic, leaveTangent=0.2132, time=0.3441, value=2}
           ]}
          GreenCurve {keys [
            {arriveTangent=-0.4331, interpMode=RCIM_Cubic, leaveTangent=-0.4331, tangentMode=RCTM_User, time=0.1298, value=1.7044}
            {interpMode=RCIM_Cubic, leaveTangent=2.0123, time=0.3441, value=1.3715}
           ]}
          RedCurve {keys [
            {arriveTangent=-4.577, interpMode=RCIM_Cubic, leaveTangent=-4.577, tangentMode=RCTM_User, time=0.1298, value=0.2261}
            {interpMode=RCIM_Cubic, leaveTangent=-2.7231, time=0.3441, value=0.1146}
           ]}
         }}))
      1=&MakeVectorFromLinearColorRGB(Color=&ColorFromCurve(ColorCurve=DataInterface() {propertyValues {
          AlphaCurve {keys []}
          BlueCurve {keys [
            {arriveTangent=-1.8762, interpMode=RCIM_Cubic, leaveTangent=-1.8762, tangentMode=RCTM_User, time=0.1298, value=0.2261}
            {interpMode=RCIM_Cubic, leaveTangent=0.2132, time=0.3441, value=0.1146}
           ]}
          GreenCurve {keys [
            {arriveTangent=-0.4331, interpMode=RCIM_Cubic, leaveTangent=-0.4331, tangentMode=RCTM_User, time=0.1298, value=1.2609}
            {interpMode=RCIM_Cubic, leaveTangent=2.0123, time=0.3441, value=0.4288}
           ]}
          RedCurve {keys [
            {arriveTangent=-4.577, interpMode=RCIM_Cubic, leaveTangent=-4.577, tangentMode=RCTM_User, time=0.1298, value=2}
            {interpMode=RCIM_Cubic, leaveTangent=-2.7231, time=0.3441, value=2}
           ]}
         }}))
      2=?
      3=?
      Index=&MakeIntFromFloat(Float=$User.Variant)
      Num=::NewEnumerator0
     )
     "Float (Alpha)"=&Multiply_Float(
      A=&FloatFromCurve(
       FloatCurve=DataInterface() {propertyValues {Curve {keys [
           {arriveTangent=-0.9, time=0, value=0}
           {arriveTangent=-0.0998, leaveTangent=-0.0998, time=0.05, value=1}
           {arriveTangent=-1.0872, leaveTangent=-1.0872, time=0.1804, value=0.7734}
           <at 0.2985 0.937>
           <at 0.4531 0.6834>
           <at 0.5813 0.9005>
           <at 0.7511 0.6392>
           <at 0.8564 0.8>
           {arriveTangent=-1.5912, leaveTangent=-1.5912, time=1, value=0}
          ]}}}
       Scale Curve=&FN_FrontEndFloatModifier(Gameplay=1, Front End=0.5)
      )
      B=&FNTimeOfDayScalar(Night Scale=0.7, Morning Scale=0.5, Evening Scale=0.6)
     )
    )) {isSetParameters=true}
   &FN_Light_Attributes(
    Color=&ColorFromCurve(ColorCurve=DataInterface() {propertyValues {
       AlphaCurve {keys []}
       BlueCurve {keys [<at 0 0 RCIM_Cubic>,<at 0.0858 30>,<at 0.1804 10>,<at 0.9 2>,<at 1 0 RCIM_Cubic>]}
       GreenCurve {keys [<at 0 0 RCIM_Cubic>,<at 0.0858 30>,<at 0.1804 10>,<at 0.9 1.3576>,<at 1 0 RCIM_Cubic>]}
       RedCurve {keys [<at 0 0 RCIM_Cubic>,<at 0.0858 0>,<at 0.1804 0>,<at 0.9 0.0729>,<at 1 0 RCIM_Cubic>]}
      }})
    Create Color Attribute=true
    Create Exponent Attribute=true
    Create Position Attribute=true
    Create Radius Attribute=true
    Exponent=3
    Position=$Particles.SkeletalMeshLocation.SampledPosition
    Radius=&FloatFromCurve(
     FloatCurve=DataInterface() {propertyValues {Curve {keys [<at 0 0.4>,<at 0.1 1.5>,<at 0.2 1>,<at 0.5 1>,<at 0.8 0.8>,<at 1 0.4>]}}}
     Scale Curve=&Multiply_Float(A=&MakeCustomFloatFromBool(Bool=$User.bIsFrontEndPreview, True Float=40, False Float=80), B=1)
    )
    Scale Color=&FloatFromCurve(
     FloatCurve=DataInterface() {propertyValues {Curve {keys [<at 0 0.5>,<at 0.1 1.1>,<at 0.1401 0.4933>,<at 0.2 1.3742>,<at 0.2354 0.5667>,<at 0.3 1.1>,<at 0.3404 0.4667>,<at 0.4098 1.3991>,<at 0.4342 0.48>,<at 0.5 1.1>,<at 0.5311 0.5133>,<at 0.5829 1.3306>,<at 0.6376 0.5133>,<at 0.7 1.1>,<at 0.741 0.5067>,<at 0.7989 1.4926>,<at 0.8459 0.3133>,<at 0.9 1.1>,<at 1 0>]}}}
     Scale Curve=&Multiply_Float(A=5, B=&FNTimeOfDayScalar(Night Scale=0.3, Morning Scale=0.5, Day Scale=0.8, Evening Scale=0.4))
    )
   )
   &SolveForcesAndVelocity(Clamp Velocity=false, Limit Acceleration=false)
  ]
  rendererClasses [
   &NiagaraSpriteRendererProperties
   &NiagaraMeshRendererProperties
   &NiagaraLightRendererProperties
  ]
  renderers [
   {rendererClass=&NiagaraSpriteRendererProperties}
   {rendererClass=&NiagaraSpriteRendererProperties, rendererIndex=1}
   {rendererClass=&NiagaraSpriteRendererProperties, rendererIndex=2}
   {rendererClass=&NiagaraMeshRendererProperties, rendererIndex=3}
   {rendererClass=&NiagaraLightRendererProperties, rendererIndex=4}
  ]
 }
 Ice_Shards {
  emitterUpdateScript [
   &EmitterState(Life Cycle Mode=::NewEnumerator1, Loop Behavior=::NewEnumerator1, Loop Delay=0.6, Loop Duration=0.77, UseLoopDelay=true)
   &SpawnRate(SpawnRate=100, Use Spawn Probability=false)
   &SpawnBurst_Instantaneous(Spawn Count=25, Use Spawn Probability=false, UseLoopCountLimit=false)
   /SetVariables(Emitter.Hue Shift:NiagaraFloat=&SelectFloatFromArray(
     Direct Array Index=&MakeIntFromFloat(Float=$User.Variant)
     Float Selection Array=DataInterface() {propertyValues {FloatData [0,0.524]}}
    )) {isSetParameters=true}
  ]
  particleSpawnScript [
   &PartitionParticles()
   &InitializeParticle(
    Lifetime Max=0.25
    Lifetime Min=0.2
    Lifetime Mode=::NewEnumerator1
    Mesh Scale Max=<0.6 0.6 1>
    Mesh Scale Min=<0.4 0.5 0.5>
    Mesh Scale Mode=::NewEnumerator2
    Position Offset=<0 -7 -5>
    Sprite Size Mode=::NewEnumerator3
    Uniform Sprite Size=&RandomRangeFloat(Minimum=1, Maximum=2)
    UsePositionOffset=true
    Write Material Random=true
   ) {module=InitializeParticle001}
   &SkeletalMeshLocation(
    Bone / Socket Index=$Particles.PartitionParticles.Partition
    Bone Sampling Mode=::NewEnumerator4
    Interpolate Mesh Sampling=false
    MaskSpawnGroup=false
    Orientation Sampling=::NewEnumerator1
    Sampled Position Offset=<0 10 0>
    Skeletal Mesh=$System.Skeletal Mesh Reader
   )
   &AddVelocity(
    Cone Axis=&SelectVectorFromArray(
     Direct Array Index=$Particles.PartitionParticles.Partition
     Vector Selection Array=DataInterface() {propertyValues {FloatData [<0.15 0 -1>, <-0.15 0 -1>]}}
    )
    Module.UseConeFalloff=false
    Module.UseVelDistribution=false
    Override Seed=false
    Velocity Mode=::NewEnumerator2
    Velocity Speed=&RandomRangeFloat(Minimum=250, Maximum=700)
   )
  ]
  particleUpdateScript [
   &ParticleState()
   &ScaleMeshSize(Scale Factor=&MakeVector(
     X=&FloatFromCurve(FloatCurve=DataInterface() {propertyValues {Curve {defaultValue=1, keys [{tangentMode=RCTM_User, time=0, value=1}, {tangentMode=RCTM_User, time=1, value=2}]}}})
     Y=&FloatFromCurve(FloatCurve=DataInterface() {propertyValues {Curve {keys [<at 0 0 RCIM_Cubic>,<at 0.5 1 RCIM_Cubic>,<at 1 0>]}}}, Scale Curve=0.7)
     Z=&FloatFromCurve(FloatCurve=DataInterface() {propertyValues {Curve {keys [<at 0 0 RCIM_Cubic>,<at 0.5 1 RCIM_Cubic>,<at 1 0>]}}}, Scale Curve=0.7)
    )) {module=ScaleMeshSize001}
   &Drag(Drag=&RandomRangeFloat(Minimum=2, Maximum=3), Use Linear Drag=true, Use Rotational Drag=false)
   &SolveForcesAndVelocity(Clamp Velocity=false, Limit Acceleration=false)
   &ScaleColor(
    Color Scale:LinearColor=&FNTimeOfDayColor(Night Color=<rgba 0.5 0.5 0.5 1>, Morning Color=<rgba 0.5 0.5 0.5 1>, Day Color=<rgba 0.5 0.5 0.5 1>, Evening Color=<rgba 0.5 0.5 0.5 1>)
    Scale Mode=3
   ) {module=ScaleColor002}
  ]
  rendererClasses [&NiagaraMeshRendererProperties]
  renderers [{rendererClass=&NiagaraMeshRendererProperties, rendererIndex=0}]
 }
 Icy_Mist {
  emitterUpdateScript [
   &EmitterState(Life Cycle Mode=::NewEnumerator1, Loop Behavior=::NewEnumerator1, Loop Delay=0.6, Loop Duration=0.62, UseLoopDelay=true)
   &SpawnRate(SpawnRate=35, Use Spawn Probability=false)
   &SpawnRate(SpawnGroup=1, SpawnRate=40, Use Spawn Probability=false) {module=SpawnRate001}
  ]
  particleSpawnScript [
   /SetVariables(Particles.VisibilityTag=&SpawnGroup()) {isSetParameters=true}
   &PartitionParticles()
   &InitializeParticle(
    Color=&MergeHelper_ColorSwitch(
     0=&RandomRangeLinearColor(Minimum=<rgba 0.5972 0.7837 0.9323 0.5>, Maximum=<rgba 0.1892 0.2908 0.3906 0.5>)
     1=&RandomRangeLinearColor(Minimum=<rgba 0.9323 0.7927 0.5972 0.5>, Maximum=<rgba 0.3906 0.2228 0.1892 0.5>)
     2=?
     3=?
     Index=&MakeIntFromFloat(Float=$User.Variant)
     Num=::NewEnumerator0
    )
    Lifetime=&RandomRangeFloat(Minimum=0.5, Maximum=1)
    Sprite Rotation Mode=::NewEnumerator2
    Sprite Size Mode=::NewEnumerator3
    Sprite UV Mode=::NewEnumerator3
    Uniform Sprite Size=&MergeHelper_FloatSwitch(
     0=&RandomRangeFloat(Minimum=25.6, Maximum=40)
     1=&RandomRangeFloat(Minimum=40, Maximum=60)
     2=?
     3=?
     Index=&SpawnGroup()
     Num=::NewEnumerator0
    )
    UsePositionOffset=false
    Write Material Random=true
   )
   &SkeletalMeshLocation(
    Bone / Socket Index=$Particles.PartitionParticles.Partition
    Bone Sampling Mode=::NewEnumerator4
    Interpolate Mesh Sampling=false
    MaskSpawnGroup=false
    Orientation Sampling=::NewEnumerator1
    Skeletal Mesh=$System.Skeletal Mesh Reader
   )
   &AddVelocity(
    Cone Angle=25
    Cone Axis=&SelectVectorFromArray(
     Direct Array Index=$Particles.PartitionParticles.Partition
     Vector Selection Array=DataInterface() {propertyValues {FloatData [<0.15 0 -1>, <-0.15 0 -1>]}}
    )
    Module.UseConeFalloff=false
    Module.UseVelDistribution=false
    Override Seed=false
    Velocity Mode=::NewEnumerator2
    Velocity Speed=&RandomRangeFloat(Minimum=0, Maximum=250)
   ) {module=AddVelocity001}
   &CameraOffset(Camera Offset Amount=&MergeHelper_FloatSwitch(0=5, 1=0, 2=?, 3=?, Index=&SpawnGroup(), Num=::NewEnumerator0))
   &DynamicMaterialParameters(
    Index 0 Param 1=&MergeHelper_FloatSwitch(
     0=3
     1=&RandomRangeFloat(Minimum=-0.05, Maximum=0.05)
     2=?
     3=?
     Index=&SpawnGroup()
     Num=::NewEnumerator0
    )
    Index 0 Param 2=&MergeHelper_FloatSwitch(
     0=1
     1=&RandomRangeFloat(Minimum=0, Maximum=32)
     2=?
     3=?
     Index=&SpawnGroup()
     Num=::NewEnumerator0
    )
    Index 0 Param 3=&MergeHelper_FloatSwitch(0=0.5, 1=0, 2=?, 3=?, Index=&SpawnGroup(), Num=::NewEnumerator0)
    Module.Index0Float0Write=true
    Module.Index0Float1Write=true
    Module.Index0Float2Write=true
    Module.Index0Float3Write=true
   )
  ]
  particleUpdateScript [
   &ParticleState()
   &Drag(
    Drag=&MergeHelper_FloatSwitch(
     0=&RandomRangeFloat(Minimum=1.5, Maximum=2.5)
     1=&RandomRangeFloat(Minimum=1, Maximum=1.5)
     2=?
     3=?
     Index=&SpawnGroup()
     Num=::NewEnumerator0
    )
    Use Linear Drag=true
    Use Rotational Drag=true
   )
   &SolveForcesAndVelocity(Clamp Velocity=false, Limit Acceleration=false)
   &ScaleSpriteSize(
    Scale Sprite Size Mode=::NewEnumerator0
    Uniform Scale Factor=&MergeHelper_FloatSwitch(
     0=&FloatFromCurve(FloatCurve=DataInterface() {propertyValues {Curve {keys [<at 0 0.5 RCIM_Cubic>, {tangentMode=RCTM_User, time=1, value=1}]}}}, Scale Curve=2)
     1=&FloatFromCurve(FloatCurve=DataInterface() {propertyValues {Curve {keys [
          {arriveTangent=2.2332, interpMode=RCIM_Cubic, leaveTangent=2.2332, tangentMode=RCTM_User, time=0, value=0.6}
          {tangentMode=RCTM_User, time=1, value=1.25}
         ]}}})
     2=?
     3=?
     Index=&SpawnGroup()
     Num=::NewEnumerator0
    )
   )
   &ScaleSpriteSizeBySpeed(Max Scale Factor=&MergeHelper_Vector2Switch(
     0=&RandomRangeVector2D(Minimum=<1 3>, Maximum=<1 7>)
     1=&RandomRangeVector2D(Minimum=<1 3>, Maximum=<1 7>)
     2=?
     3=?
     4=?
     Index=&SpawnGroup()
     Num=::NewEnumerator0
    ))
   &SubUVAnimation(
    "Sprite Renderer "=DataInterface() {propertyValues {SpriteRenderer=&self:Icy_Mist_0.NiagaraSpriteRendererProperties_0}}
    SubUV Animation Mode:ENiagara_SubUVLookupModeV2=::NewEnumerator1
    SubUV Curve=DataInterface() {propertyValues {Curve {keys [
        {arriveTangent=1.5019, interpMode=RCIM_Cubic, leaveTangent=1.5019, tangentMode=RCTM_User, time=0, value=0}
        {arriveTangent=0.2086, interpMode=RCIM_Cubic, leaveTangent=0.2086, tangentMode=RCTM_User, time=1, value=1}
       ]}}}
    UseEndFrame=false
    UseStartFrame=false
   )
   &ScaleColor(
    Scale Alpha=&MergeHelper_FloatSwitch(
     0=&FloatFromCurve(
      FloatCurve=DataInterface() {propertyValues {Curve {keys [
          {interpMode=RCIM_Cubic, tangentMode=RCTM_User, time=0, value=0}
          {interpMode=RCIM_Cubic, tangentMode=RCTM_User, time=0.5, value=1}
          {tangentMode=RCTM_User, time=1, value=0}
         ]}}}
      Scale Curve=&FN_FrontEndFloatModifier(Gameplay=0.35, Front End=0.25)
     )
     1=&Multiply_Float(
      A=&FN_FrontEndFloatModifier(
       Gameplay=&Multiply_Float(
        A=&FloatFromCurve(
         FloatCurve=DataInterface() {propertyValues {Curve {keys [
             {arriveTangent=9.0725, interpMode=RCIM_Cubic, leaveTangent=9.0725, tangentMode=RCTM_User, time=0, value=0}
             {arriveTangent=-0.5016, interpMode=RCIM_Cubic, leaveTangent=-0.5016, tangentMode=RCTM_User, time=0.1408, value=0.9847}
             {arriveTangent=-0.0405, interpMode=RCIM_Cubic, leaveTangent=-0.0405, tangentMode=RCTM_User, time=1, value=0}
            ]}}}
         Scale Curve=0.35
        )
        B=&FloatFromCurve(
         FloatCurve=DataInterface() {propertyValues {Curve {keys [
             {interpMode=RCIM_Cubic, tangentMode=RCTM_User, time=0, value=0.7}
             {interpMode=RCIM_Cubic, tangentMode=RCTM_User, time=0.5, value=1.3}
             {tangentMode=RCTM_User, time=1, value=0.7}
            ]}}}
         CurveIndex=Linked() {linkedVariable {name=NPC.FortniteNPC.TimeOfDay_0-1, type=NiagaraFloat}}
        )
       )
       Front End=&Multiply_Float(
        A=&FloatFromCurve(
         FloatCurve=DataInterface() {propertyValues {Curve {keys [
             {arriveTangent=9.0725, interpMode=RCIM_Cubic, leaveTangent=9.0725, tangentMode=RCTM_User, time=0, value=0}
             {arriveTangent=-0.5016, interpMode=RCIM_Cubic, leaveTangent=-0.5016, tangentMode=RCTM_User, time=0.1408, value=0.9847}
             {arriveTangent=-0.0405, interpMode=RCIM_Cubic, leaveTangent=-0.0405, tangentMode=RCTM_User, time=1, value=0}
            ]}}}
         Scale Curve=0.75
        )
        B=0.25
       )
      )
      B=&FN_FrontEndFloatModifier(
       Gameplay=&SelectFloatFromArray(Direct Array Index=$Engine.QualityLevel, Float Selection Array=DataInterface() {propertyValues {FloatData [0.5,0.5,1,1,1]}})
       Front End=&SelectFloatFromArray(Direct Array Index=$Engine.QualityLevel, Float Selection Array=DataInterface() {propertyValues {FloatData [0.5,1,1,1,1]}})
      )
     )
     2=?
     3=?
     Index=&SpawnGroup()
     Num=::NewEnumerator0
    )
    ScaleA=true
    ScaleRGB=true
   )
   &KillParticles(Kill Particles=&Boolean_NotOperation(Bool=/Game/Effects/Niagara/DynamicInputs/State/FN_IsExecutionStateActive()))
   /SetVariables() {isSetParameters=true}
  ]
  rendererClasses [&NiagaraSpriteRendererProperties]
  renderers [
   {rendererClass=&NiagaraSpriteRendererProperties}
   {rendererClass=&NiagaraSpriteRendererProperties, rendererIndex=1}
  ]
 }
 Sec_Energy {
  emitterUpdateScript [
   &EmitterState(
    Inactive Response=::NewEnumerator1
    Life Cycle Mode=::NewEnumerator1
    Loop Behavior=::NewEnumerator1
    Loop Delay=0.2
    Loop Duration=1.12
    UseLoopDelay=true
   )
   &SpawnBurst_Instantaneous(Spawn Count=3, Use Spawn Probability=false, UseLoopCountLimit=true)
   &SpawnRate(SpawnRate=50, Use Spawn Probability=false)
  ]
  particleSpawnScript [
   &PartitionParticles()
   &InitializeParticle(
    Color Maximum=<rgba 0.05 0.0463 0.0425 1>
    Color Minimum=<rgba 0.15 0.1388 0.1275 1.5>
    Color Mode=::NewEnumerator2
    Lifetime Max=0.3
    Lifetime Min=0.25
    Lifetime Mode=::NewEnumerator1
    Mesh Scale=&Multiply_VectorByFloat(
     Vector=&RandomRangeVector(Minimum=<1 -1.5 -1.5>, Maximum=<1.1 1.5 1.5>)
     Float=&FN_FrontEndFloatModifier(Gameplay=0.75, Front End=1)
    )
    Mesh Scale Mode=::NewEnumerator4
    UsePositionOffset=true
    Write Material Random=true
   )
   &AddVelocity(Velocity=&RandomRangeVector(Minimum=<-15 -5 -20>, Maximum=<15 5 -20>))
   &InitialMeshOrientation(
    Mesh Orientation Mode=::NewEnumerator2
    Rotation=&RandomRangeVector(Minimum=<0 0 -0.15>, Maximum=<0 0 0.15>)
    Use Rotation Vector=true
   )
  ]
  particleUpdateScript [
   &ParticleState()
   &SkeletalMeshLocation(
    Bone / Socket Index=$Particles.PartitionParticles.Partition
    Bone Sampling Mode=::NewEnumerator4
    MaskSpawnGroup=false
    Orientation Sampling=::NewEnumerator1
    Skeletal Mesh=$System.Skeletal Mesh Reader
   )
   &ScaleMeshSize(Scale Factor=&Multiply_Vector(
     A=&MakeVector(
      X=&FloatFromCurve(FloatCurve=DataInterface() {propertyValues {Curve {keys [<at 0 0.4>,<at 0.1396 0.8>,<at 1 1>]}}}, Scale Curve=1.2)
      Y=1
      Z=1
     )
     B=&Divide_Vector(A:Vector3f=<1 1 1>, B:Vector3f=$Engine.Owner.Scale)
    ))
   &ScaleColor(
    Scale Alpha=&Multiply_Float(
     A=&FloatFromCurve(
      FloatCurve=DataInterface() {propertyValues {Curve {keys [
          {arriveTangent=2.8247, interpMode=RCIM_Cubic, leaveTangent=2.8247, tangentMode=RCTM_User, time=0, value=0.5}
          {arriveTangent=0.0384, interpMode=RCIM_Cubic, leaveTangent=0.0384, tangentMode=RCTM_User, time=0.4, value=1}
          {arriveTangent=0.0424, interpMode=RCIM_Cubic, leaveTangent=0.0424, tangentMode=RCTM_User, time=1, value=0}
         ]}}}
      Scale Curve=&FNTimeOfDayScalar(Night Scale=0.6, Morning Scale=0.8, Day Scale=1.2, Evening Scale=0.6)
     )
     B=&FN_FrontEndFloatModifier(Gameplay=1, Front End=0.6)
    )
    Scale RGB=&MakeVectorFromLinearColorRGB(Color=&MergeHelper_ColorSwitch(
      0=&ColorFromCurve(ColorCurve=DataInterface() {propertyValues {
         AlphaCurve {keys []}
         BlueCurve {keys [
           {arriveTangent=-6.2021, interpMode=RCIM_Cubic, leaveTangent=-6.2021, tangentMode=RCTM_User, time=0.1852, value=15}
           {arriveTangent=-1.9715, interpMode=RCIM_Cubic, leaveTangent=-1.9715, tangentMode=RCTM_User, time=0.4613, value=5}
           {interpMode=RCIM_Cubic, leaveTangent=-0.0154, time=0.8539, value=0.1}
          ]}
         GreenCurve {keys [
           {arriveTangent=-6.099, interpMode=RCIM_Cubic, leaveTangent=-6.099, tangentMode=RCTM_User, time=0.1852, value=15}
           {arriveTangent=0.1945, interpMode=RCIM_Cubic, leaveTangent=0.1945, tangentMode=RCTM_User, time=0.4613, value=4.1667}
           {arriveTangent=-0.0129, interpMode=RCIM_Cubic, leaveTangent=-0.0129, tangentMode=RCTM_User, time=0.8539, value=0.0667}
          ]}
         RedCurve {keys [
           {arriveTangent=-31.4411, interpMode=RCIM_Cubic, leaveTangent=-31.4411, tangentMode=RCTM_User, time=0.1852, value=2.0312}
           {arriveTangent=-21.5603, interpMode=RCIM_Cubic, leaveTangent=-21.5603, tangentMode=RCTM_User, time=0.4613, value=0}
           {interpMode=RCIM_Cubic, leaveTangent=-0.0154, time=0.8539, value=0}
          ]}
        }})
      1=&ColorFromCurve(ColorCurve=DataInterface() {propertyValues {
         AlphaCurve {keys []}
         BlueCurve {keys [
           {arriveTangent=-6.2021, interpMode=RCIM_Cubic, leaveTangent=-6.2021, tangentMode=RCTM_User, time=0.1852, value=2.0313}
           {arriveTangent=-1.9715, interpMode=RCIM_Cubic, leaveTangent=-1.9715, tangentMode=RCTM_User, time=0.4613, value=0}
           {interpMode=RCIM_Cubic, leaveTangent=-0.0154, time=0.8539, value=0}
          ]}
         GreenCurve {keys [
           {arriveTangent=-6.099, interpMode=RCIM_Cubic, leaveTangent=-6.099, tangentMode=RCTM_User, time=0.1852, value=9.5964}
           {arriveTangent=0.1945, interpMode=RCIM_Cubic, leaveTangent=0.1945, tangentMode=RCTM_User, time=0.4613, value=2.9167}
           {arriveTangent=-0.0129, interpMode=RCIM_Cubic, leaveTangent=-0.0129, tangentMode=RCTM_User, time=0.8539, value=0.0167}
          ]}
         RedCurve {keys [
           {arriveTangent=-31.4411, interpMode=RCIM_Cubic, leaveTangent=-31.4411, tangentMode=RCTM_User, time=0.1852, value=15}
           {arriveTangent=-21.5603, interpMode=RCIM_Cubic, leaveTangent=-21.5603, tangentMode=RCTM_User, time=0.4613, value=5}
           {interpMode=RCIM_Cubic, leaveTangent=-0.0154, time=0.8539, value=0.1}
          ]}
        }})
      2=?
      3=?
      Index=&MakeIntFromFloat(Float=$User.Variant)
      Num=::NewEnumerator0
     ))
    ScaleA=true
    ScaleRGB=true
   )
   &Drag(Drag=6, Use Linear Drag=true, Use Rotational Drag=false)
   &SolveForcesAndVelocity(Clamp Velocity=false, Limit Acceleration=false)
  ]
  rendererClasses [&NiagaraMeshRendererProperties]
  renderers [{rendererClass=&NiagaraMeshRendererProperties, rendererIndex=0}]
 }
}
summary {
 systemName=NS_Glider_BriskReign_Beams
 emitters [
  {emitterName=Beams, rendererClasses [&NiagaraMeshRendererProperties]}
  {
   emitterName=Glows
   rendererClasses [
    &NiagaraSpriteRendererProperties
    &NiagaraMeshRendererProperties
    &NiagaraLightRendererProperties
   ]
  }
  {emitterName=Sec_Energy, rendererClasses [&NiagaraMeshRendererProperties]}
  {emitterName=Icy_Mist, rendererClasses [&NiagaraSpriteRendererProperties]}
  {emitterName=Ice_Shards, rendererClasses [&NiagaraMeshRendererProperties]}
 ]
 userVariables [
  User.bIsFrontEnd {type=NiagaraBool}
  User.bIsFrontEndPreview {type=NiagaraBool}
  User.Variant {type=NiagaraFloat, defaultValue=0, description=""}
 ]
}
