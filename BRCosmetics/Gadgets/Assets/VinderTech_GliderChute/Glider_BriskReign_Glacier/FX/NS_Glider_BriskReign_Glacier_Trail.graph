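# Niagara trail system for the BriskReign Glacier glider, with dark and light AnimTrail ribbons, ice shard meshes and icy mist sprites.
# All emitters attach to skeletal mesh bones, and ribbons also shift with owner velocity.
# System timelines ramp trail velocity, opacity and spawn rates, with separate front-end and gameplay inputs.
# User.Variant switches between cool and warm palettes, ForwardDot feeds ribbon material parameters, and bIsFullyDeployed gates mist spawning.
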NiagaraSystem /BRCosmetics/Gadgets/Assets/VinderTech_GliderChute/Glider_BriskReign_Glacier/FX/NS_Glider_BriskReign_Glacier_Trail
dependencies {
 usedDataInterfaces [
  &NiagaraDataInterfaceCurve
  &NiagaraDataInterfaceSpriteRendererInfo
  &NiagaraDataInterfaceArrayFloat3
  &NiagaraDataInterfaceArrayFloat
  &NiagaraDataInterfaceColorCurve
  &NiagaraDataInterfaceSkeletalMesh
 ]
 usedDynamicInputs [
  &Vector2DFromFloat
  &SetBoolByFloatComparison
  &RemapRange
 ]
 usedModules [
  &self:SystemSpawnScript.SystemScriptSource.SystemScriptGraph.NiagaraNodeAssignment_0.SetVariables_#1
  &SystemState
  &Timeline
 ]
}
emitters {
 FN_AnimTrail_Dark {
  emitterSpawnScript [&FN_RibbonIDOffsetUtility()]
  emitterUpdateScript [
   &EmitterState(
    Inactive Response=::NewEnumerator1
    Life Cycle Mode=::NewEnumerator1
    Loop Behavior=::NewEnumerator1
    Loop Delay=0.6
    Loop Duration Mode=::NewEnumerator1
    UseLoopDelay=true
   )
   &FN_SpawnAnimTrail(Number of Anim Trails=6, Switch on ENiagara_AnimTrail_Source:ENiagara_AnimTrail_Source=::NewEnumerator1)
   /SetVariables(Emitter.WPOIntensity:NiagaraFloat=&FloatFromCurve(
     FloatCurve=DataInterface() {propertyValues {Curve {
        defaultValue=1
        keys [{interpMode=RCIM_Cubic, tangentMode=RCTM_User, time=0, value=1}, {interpMode=RCIM_Cubic, tangentMode=RCTM_User, time=1, value=0}]
       }}}
     CurveIndex=$Emitter.NormalizedLoopAge
     Scale Curve=5
    )) {isSetParameters=true}
  ]
  particleSpawnScript [
   &FN_InitializeAnimTrailBoneIndecies(First Bone or Socket Index=2, Second Bone or Socket Index=2) {module=FN_InitializeAnimTrailBoneIndecies003}
   &FN_InitializeAnimTrailBoneIndecies(AnimTrail Selection=1, First Bone or Socket Index=3, Second Bone or Socket Index=3) {module=FN_InitializeAnimTrailBoneIndecies002}
   &FN_InitializeAnimTrailBoneIndecies(AnimTrail Selection=2, First Bone or Socket Index=4, Second Bone or Socket Index=4) {module=FN_InitializeAnimTrailBoneIndecies004}
   &FN_InitializeAnimTrailBoneIndecies(AnimTrail Selection=3, First Bone or Socket Index=7, Second Bone or Socket Index=7) {module=FN_InitializeAnimTrailBoneIndecies007}
   &FN_InitializeAnimTrailBoneIndecies(AnimTrail Selection=4, First Bone or Socket Index=8, Second Bone or Socket Index=8) {module=FN_InitializeAnimTrailBoneIndecies006}
   &FN_InitializeAnimTrailBoneIndecies(AnimTrail Selection=5, First Bone or Socket Index=9, Second Bone or Socket Index=9) {module=FN_InitializeAnimTrailBoneIndecies005}
   &FN_InitializeAnimTrail(AnimTrailWidth=75, SkeletalMesh=$System.Skeletal Mesh)
   &AddVelocity(
    Velocity=&FN_FrontEndVector3Modifier(
     Gameplay=&SelectVectorFromArray(
      Direct Array Index=&SpawnGroup()
      Vector Selection Array=DataInterface() {propertyValues {FloatData [<250 800 250>, <250 800 250>, <250 800 250>, <-100 800 275>, <-100 800 275>, <-100 800 275>]}}
     )
     Front End=&SelectVectorFromArray(
      Direct Array Index=&SpawnGroup()
      Vector Selection Array=DataInterface() {propertyValues {FloatData [<-100 800 250>, <-100 800 250>, <-100 800 250>, <100 800 275>, <100 800 275>, <100 800 275>]}}
     )
    )
    Velocity Speed Scale=&FN_FrontEndFloatModifier(
     Gameplay=&Lerp_Float(A=0, B=1, Alpha=$System.Timeline.NormalizedTime)
     Front End=&Lerp_Float(A=0, B=1, Alpha=$System.Timeline002.NormalizedTime)
    )
   ) {module=AddVelocity001}
   &AddVelocity(
    Velocity=<0 1000 0>
    Velocity Speed Scale=&FN_FrontEndFloatModifier(
     Gameplay=&Lerp_Float(A=0, B=0.7, Alpha=$System.Timeline.NormalizedTime)
     Front End=&Lerp_Float(A=0, B=1.05, Alpha=$System.Timeline002.NormalizedTime)
    )
   ) {module=AddVelocity002}
  ]
  particleUpdateScript [
   &InheritSourceMovement(
    Applied Position Scale=&Multiply_VectorByFloat(
     Vector=<1 1 1>
     Float=&FloatFromCurve(FloatCurve=DataInterface() {propertyValues {Curve {keys [
          {interpMode=RCIM_Cubic, tangentMode=RCTM_User, time=0.66, value=1}
          {arriveTangent=-1.9042, interpMode=RCIM_Cubic, leaveTangent=-1.9042, tangentMode=RCTM_User, time=1, value=0.75}
         ]}}})
    )
    ApplyForce=false
    ApplyPosition=true
    ApplyPositionFalloff=false
   ) {enabled=false}
   &SolveForcesAndVelocity(Clamp Velocity=false, Limit Acceleration=false)
   &FN_UpdateAnimTrail(Locked Particle RibbonWidth:NiagaraFloat=$Particles.RibbonWidth, SkeletalMesh=$System.Skeletal Mesh) {module=FN_UpdateAnimTrail001}
   /SetVariables(Particles.Position=&AddVectorToPosition(
     Position=$Particles.Position
     Vector=&Multiply_Vector(
      A=&TransformVector(Vector=$Engine.Owner.Velocity, Source Space=::World, Destination Space=::Local)
      B=&Multiply_VectorByFloat(
       Vector=<-1 0 0>
       Float=&FloatFromCurve(
        FloatCurve=DataInterface() {propertyValues {Curve {keys [
            {arriveTangent=5.4497, interpMode=RCIM_Cubic, leaveTangent=5.4497, tangentMode=RCTM_User, time=0, value=0}
            {interpMode=RCIM_Cubic, tangentMode=RCTM_User, time=0.5, value=1}
           ]}}}
        Scale Curve=0.015
       )
      )
     )
    )) {isSetParameters=true}
   &DynamicMaterialParameters(
    Index 0 Param 3=1.5
    Index 0 Param 4=&SelectFloatFromArray(
     Direct Array Index=&SpawnGroup()
     Float Selection Array=DataInterface() {propertyValues {FloatData [0,-0.1654,-3.716,-0.3475,0.0869,-0.5969,3.25]}}
    )
    Index 1 Param 1=3
    Index 1 Param 2=&Lerp_Float(
     A=1
     B=&FloatFromCurve(FloatCurve=DataInterface() {propertyValues {Curve {keys [<at 0 0>,<at 1 1>]}}}, CurveIndex=$System.Timeline.NormalizedTime)
     Alpha=$User.ForwardDot
    )
    Index 1 Param 3=3
    Module.Index0Float0Write=true
    Module.Index0Float1Write=true
    Module.Index0Float2Write=true
    Module.Index0Float3Write=true
    Module.Index1Float0Write=true
    Module.Index1Float1Write=true
    Module.Index1Float2Write=true
    Module.Index1Float3Write=true
    Write Parameter Index 1=::NewEnumerator1
   ) {module=DynamicMaterialParameters001}
   &ScaleColor(
    Scale Alpha=&FloatFromCurve(
     FloatCurve=DataInterface() {propertyValues {Curve {
        defaultValue=1
        keys [
         {interpMode=RCIM_Cubic, tangentMode=RCTM_User, time=0, value=1}
         {arriveTangent=-3.0233, interpMode=RCIM_Cubic, leaveTangent=-3.0233, tangentMode=RCTM_User, time=0.6, value=0}
        ]
       }}}
     Scale Curve=2
    )
    Scale RGB=&MakeVectorFromLinearColorRGB(Color=&ColorFromCurve(ColorCurve=DataInterface() {propertyValues {
        AlphaCurve {keys [<at 0 1 RCIM_Cubic>]}
        BlueCurve {keys [<at 0 0.0469>,<at 0.7934 0.0312>]}
        GreenCurve {keys [<at 0 0.0276>,<at 0.7934 0.0219>]}
        RedCurve {keys [<at 0 0.0147>,<at 0.7934 0.0189>]}
       }}))
    ScaleA=true
    ScaleRGB=true
   ) {module=ScaleColor002}
   &ScaleColor(
    Scale Alpha=&FN_FrontEndFloatModifier(
     Gameplay=&Lerp_Float(
      A=0
      B=0.85
      Alpha=&FloatFromCurve(FloatCurve=DataInterface() {propertyValues {Curve {keys [<at 0 0>,<at 1 1>]}}}, CurveIndex=$System.Timeline.NormalizedTime)
     )
     Front End=&Lerp_Float(
      A=0
      B=0.85
      Alpha=&FloatFromCurve(FloatCurve=DataInterface() {propertyValues {Curve {keys [<at 0 0>,<at 1 1>]}}}, CurveIndex=$System.Timeline002.NormalizedTime)
     )
    )
    ScaleA=true
    ScaleRGB=true
   )
   /SetVariables(
    Particles.Trail2 Color:LinearColor=&MakeLinearColorFromVectorAndFloat(
     "Vector (RGB)"=&MergeHelper_VectorSwitch(
      0=&MakeVectorFromLinearColorRGB(Color=<rgba 0.8333 5 5 1>)
      1=&MakeVectorFromLinearColorRGB(Color=<rgba 5 3.2639 0.8333 1>)
      2=?
      3=?
      Index=&MakeIntFromFloat(Float=$User.Variant)
      Num=::NewEnumerator0
     )
     "Float (Alpha)"=&Lerp_Float(
      A=0
      B=&FloatFromCurve(FloatCurve=DataInterface() {propertyValues {Curve {keys [<at 0 1>,<at 0.5 0>]}}}, Scale Curve=5)
      Alpha=&FloatFromCurve(FloatCurve=DataInterface() {propertyValues {Curve {keys [<at 0 0>,<at 1 1>]}}}, CurveIndex=$System.Timeline.NormalizedTime)
     )
    )
    Particles.Trail2 Width:NiagaraFloat=&Multiply_Float(A=$Particles.RibbonWidth, B=0.75)
   ) {isSetParameters=true, module=SetVariables_#2}
  ]
  rendererClasses [&NiagaraRibbonRendererProperties]
  renderers [
   {rendererClass=&NiagaraRibbonRendererProperties}
   {rendererClass=&NiagaraRibbonRendererProperties, rendererIndex=1}
   {rendererClass=&NiagaraRibbonRendererProperties, rendererIndex=2}
  ]
 }
 FN_AnimTrail_Light {
  emitterSpawnScript [&FN_RibbonIDOffsetUtility()]
  emitterUpdateScript [
   &EmitterState(
    Inactive Response=::NewEnumerator1
    Life Cycle Mode=::NewEnumerator1
    Loop Behavior=::NewEnumerator1
    Loop Delay=0.6
    Loop Duration Mode=::NewEnumerator1
    UseLoopDelay=true
   )
   &FN_SpawnAnimTrail(Number of Anim Trails=2, Switch on ENiagara_AnimTrail_Source:ENiagara_AnimTrail_Source=::NewEnumerator1)
   /SetVariables(Emitter.WPOIntensity:NiagaraFloat=&FloatFromCurve(
     FloatCurve=DataInterface() {propertyValues {Curve {
        defaultValue=1
        keys [{interpMode=RCIM_Cubic, tangentMode=RCTM_User, time=0, value=1}, {interpMode=RCIM_Cubic, tangentMode=RCTM_User, time=1, value=0}]
       }}}
     CurveIndex=$Emitter.NormalizedLoopAge
     Scale Curve=5
    )) {isSetParameters=true}
  ]
  particleSpawnScript [
   &FN_InitializeAnimTrailBoneIndecies(First Bone or Socket Index=3, Second Bone or Socket Index=3)
   &FN_InitializeAnimTrailBoneIndecies(AnimTrail Selection=1, First Bone or Socket Index=8, Second Bone or Socket Index=8) {module=FN_InitializeAnimTrailBoneIndecies001}
   &FN_InitializeAnimTrail(
    AnimTrailWidth=30
    Color=&LerpLinearColors(StartColor=<rgba 0 0.6087 1 1>, EndColor=<rgba 0.9647 0.4459 0.0754 1>, LerpFactor=$User.Variant)
    SkeletalMesh=$System.Skeletal Mesh
   )
   &AddVelocity(
    Velocity=&FN_FrontEndVector3Modifier(
     Gameplay=&SelectVectorFromArray(
      Direct Array Index=&SpawnGroup()
      Vector Selection Array=DataInterface() {propertyValues {FloatData [<250 800 250>, <-100 800 250>]}}
     )
     Front End=&SelectVectorFromArray(
      Direct Array Index=&SpawnGroup()
      Vector Selection Array=DataInterface() {propertyValues {FloatData [<-100 800 250>, <100 800 250>]}}
     )
    )
    Velocity Speed Scale=&FN_FrontEndFloatModifier(
     Gameplay=&Lerp_Float(A=0, B=1, Alpha=$System.Timeline.NormalizedTime)
     Front End=&Lerp_Float(A=0, B=1, Alpha=$System.Timeline002.NormalizedTime)
    )
   ) {module=AddVelocity001}
   &AddVelocity(
    Velocity=<0 800 0>
    Velocity Speed Scale=&FN_FrontEndFloatModifier(
     Gameplay=&Lerp_Float(A=0, B=0.665, Alpha=$System.Timeline.NormalizedTime)
     Front End=&Lerp_Float(A=0, B=1.015, Alpha=$System.Timeline002.NormalizedTime)
    )
   ) {module=AddVelocity002}
  ]
  particleUpdateScript [
   &InheritSourceMovement(
    Applied Position Scale=&Multiply_VectorByFloat(
     Vector=<1 1 1>
     Float=&FloatFromCurve(FloatCurve=DataInterface() {propertyValues {Curve {keys [
          {interpMode=RCIM_Cubic, tangentMode=RCTM_User, time=0.66, value=1}
          {arriveTangent=-1.9042, interpMode=RCIM_Cubic, leaveTangent=-1.9042, tangentMode=RCTM_User, time=1, value=0.75}
         ]}}})
    )
    ApplyForce=false
    ApplyPosition=true
    ApplyPositionFalloff=false
   ) {enabled=false}
   &SolveForcesAndVelocity(Clamp Velocity=false, Limit Acceleration=false)
   &FN_UpdateAnimTrail(Locked Particle RibbonWidth=1, SkeletalMesh=$System.Skeletal Mesh) {module=FN_UpdateAnimTrail001}
   /SetVariables(Particles.Position=&AddVectorToPosition(
     Position=$Particles.Position
     Vector=&Multiply_Vector(
      A=&TransformVector(Vector=$Engine.Owner.Velocity, Source Space=::World, Destination Space=::Local)
      B=&Multiply_VectorByFloat(
       Vector=<-1 0 0>
       Float=&FloatFromCurve(
        FloatCurve=DataInterface() {propertyValues {Curve {keys [
            {arriveTangent=5.4497, interpMode=RCIM_Cubic, leaveTangent=5.4497, tangentMode=RCTM_User, time=0, value=0}
            {interpMode=RCIM_Cubic, tangentMode=RCTM_User, time=0.5, value=1}
           ]}}}
        Scale Curve=0.015
       )
      )
     )
    )) {isSetParameters=true}
   &DynamicMaterialParameters(
    Index 0 Param 1=150
    Index 0 Param 2=60
    Index 0 Param 3=1.5
    Index 0 Param 4=&SelectFloatFromArray(
     Direct Array Index=&SpawnGroup()
     Float Selection Array=DataInterface() {propertyValues {FloatData [1.153,0.3543]}}
    )
    Index 1 Param 1=3
    Index 1 Param 2=&Lerp_Float(
     A=1
     B=&FloatFromCurve(FloatCurve=DataInterface() {propertyValues {Curve {keys [<at 0 0>,<at 1 1>]}}}, CurveIndex=$System.Timeline.NormalizedTime)
     Alpha=$User.ForwardDot
    )
    Index 1 Param 3=3
    Module.Index0Float0Write=true
    Module.Index0Float1Write=true
    Module.Index0Float2Write=true
    Module.Index0Float3Write=true
    Module.Index1Float0Write=true
    Module.Index1Float1Write=true
    Module.Index1Float2Write=true
    Module.Index1Float3Write=true
    Write Parameter Index 1=::NewEnumerator1
   ) {module=DynamicMaterialParameters001}
   &ScaleColor(
    Scale Alpha=&FloatFromCurve(
     FloatCurve=DataInterface() {propertyValues {Curve {
        defaultValue=1
        keys [
         {interpMode=RCIM_Cubic, tangentMode=RCTM_User, time=0, value=1}
         {arriveTangent=-3.0233, interpMode=RCIM_Cubic, leaveTangent=-3.0233, tangentMode=RCTM_User, time=0.5, value=0}
        ]
       }}}
     Scale Curve=2
    )
    ScaleA=true
    ScaleRGB=true
   ) {module=ScaleColor002}
   &ScaleColor(
    Scale Alpha=&FN_FrontEndFloatModifier(
     Gameplay=&Lerp_Float(
      A=0
      B=0.75
      Alpha=&FloatFromCurve(FloatCurve=DataInterface() {propertyValues {Curve {keys [<at 0 0>,<at 1 1>]}}}, CurveIndex=$System.Timeline.NormalizedTime)
     )
     Front End=&Lerp_Float(
      A=0
      B=0.75
      Alpha=&FloatFromCurve(FloatCurve=DataInterface() {propertyValues {Curve {keys [<at 0 0>,<at 1 1>]}}}, CurveIndex=$System.Timeline002.NormalizedTime)
     )
    )
    ScaleA=true
    ScaleRGB=true
   )
   /SetVariables(Particles.Linear Color=&MakeLinearColorFromVectorAndFloat(
     "Vector (RGB)"=&MergeHelper_VectorSwitch(
      0=&MakeVectorFromLinearColorRGB(Color=<rgba 0.8333 5 5 1>)
      1=&MakeVectorFromLinearColorRGB(Color=<rgba 5 3.2639 0.8333 1>)
      2=?
      3=?
      Index=&MakeIntFromFloat(Float=$User.Variant)
      Num=::NewEnumerator0
     )
     "Float (Alpha)"=&Lerp_Float(
      A=0
      B=1
      Alpha=&FloatFromCurve(FloatCurve=DataInterface() {propertyValues {Curve {keys [<at 0 0>,<at 1 1>]}}}, CurveIndex=$System.Timeline.NormalizedTime)
     )
    )) {isSetParameters=true, module=SetVariables_#2}
  ]
  rendererClasses [&NiagaraRibbonRendererProperties]
  renderers [
   {rendererClass=&NiagaraRibbonRendererProperties}
   {rendererClass=&NiagaraRibbonRendererProperties, rendererIndex=1}
   {rendererClass=&NiagaraRibbonRendererProperties, rendererIndex=2}
  ]
 }
 Ice_Shards {
  emitterUpdateScript [
   &EmitterState(Life Cycle Mode=::NewEnumerator1, Loop Behavior=::NewEnumerator1, Loop Delay=0.6, Loop Duration Mode=::NewEnumerator1, UseLoopDelay=true)
   &SpawnRate(SpawnRate=&Lerp_Float(A=0, B=22.5, Alpha=$System.Timeline.Time), Use Spawn Probability=false)
   /SetVariables(Emitter.Hue Shift:NiagaraFloat=&SelectFloatFromArray(
     Direct Array Index=&MakeIntFromFloat(Float=$User.Variant)
     Float Selection Array=DataInterface() {propertyValues {FloatData [0,0.524]}}
    )) {isSetParameters=true}
  ]
  particleSpawnScript [
   &PartitionParticles()
   &InitializeParticle(
    Lifetime Max=0.6
    Lifetime Min=0.3
    Lifetime Mode=::NewEnumerator1
    Mesh Scale Max=<0.6 0.6 1>
    Mesh Scale Min=<0.4 0.5 0.5>
    Mesh Scale Mode=::NewEnumerator2
    Position Offset=<0 -7 -5>
    Sprite Size Mode=::NewEnumerator3
    Uniform Sprite Size=&RandomRangeFloat(Minimum=1, Maximum=2)
    UsePositionOffset=true
    Write Material Random=true
   ) {module=InitializeParticle001}
   &SkeletalMeshLocation(
    Bone / Socket Index=$Particles.PartitionParticles.Partition
    Bone Sampling Mode=::NewEnumerator4
    Interpolate Mesh Sampling=false
    MaskSpawnGroup=false
    Orientation Sampling=::NewEnumerator1
    Skeletal Mesh=$System.Skeletal Mesh
   )
   &SphereLocation(
    MaskSpawnGroup=false
    Non Uniform Scale=<1 5 3>
    Offset=&SelectVectorFromArray(
     Direct Array Index=$Particles.PartitionParticles.Partition
     Vector Selection Array=DataInterface() {propertyValues {FloatData [<-10 50 0>, <10 50 0>]}}
    )
    Override Seed=false
    Sphere Origin:NiagaraPosition=$Particles.SkeletalMeshLocation.SampledPosition
    Sphere Radius=15
    Surface Only=true
    UseNonUniformScale=true
    UseOffset=true
    UseSphereOrientation=false
   )
   &AddVelocity(Velocity=&FN_FrontEndVector3Modifier(
     Gameplay=&Multiply_Vector(
      A=&MergeHelper_VectorSwitch(
       0=&RandomRangeVector(Minimum=<-100 150 250>, Maximum=<-100 250 500>)
       1=&RandomRangeVector(Minimum=<100 150 250>, Maximum=<100 250 500>)
       2=?
       3=?
       Index=$Particles.PartitionParticles.Partition
       Num=::NewEnumerator0
      )
      B=&Lerp_Vector(A=<1 3 1>, B=<1 3 1>, Alpha=$System.Timeline.NormalizedTime)
     )
     Front End=&Multiply_Vector(
      A=&MergeHelper_VectorSwitch(
       0=&RandomRangeVector(Minimum=<-100 150 250>, Maximum=<-100 250 500>)
       1=&RandomRangeVector(Minimum=<100 150 250>, Maximum=<100 250 500>)
       2=?
       3=?
       Index=$Particles.PartitionParticles.Partition
       Num=::NewEnumerator0
      )
      B=&Lerp_Vector(A=<1 3 1>, B=<1 3 1>, Alpha=$System.Timeline.NormalizedTime)
     )
    )) {module=AddVelocity001}
  ]
  particleUpdateScript [
   &ParticleState()
   &ScaleMeshSize(Scale Factor=&MakeVector(
     X=&FloatFromCurve(FloatCurve=DataInterface() {propertyValues {Curve {defaultValue=1, keys [{tangentMode=RCTM_User, time=0, value=1}, {tangentMode=RCTM_User, time=1, value=2}]}}})
     Y=&FloatFromCurve(FloatCurve=DataInterface() {propertyValues {Curve {keys [<at 0 0 RCIM_Cubic>,<at 0.5 1 RCIM_Cubic>,<at 1 0>]}}}, Scale Curve=0.7)
     Z=&FloatFromCurve(FloatCurve=DataInterface() {propertyValues {Curve {keys [<at 0 0 RCIM_Cubic>,<at 0.5 1 RCIM_Cubic>,<at 1 0>]}}}, Scale Curve=0.7)
    )) {module=ScaleMeshSize001}
   &Drag(Drag=&RandomRangeFloat(Minimum=1, Maximum=2), Use Linear Drag=true, Use Rotational Drag=false)
   &GravityForce()
   &SolveForcesAndVelocity(Clamp Velocity=false, Limit Acceleration=false)
   &ScaleColor(
    Scale RGB=&MakeVectorFromLinearColorRGB(Color=&FN_Select_Color_from_Time_Of_Day(Morning=<rgba 1 1 1 1>, Day=<rgba 1 1 1 1>, Evening=<rgba 1 1 1 1>, Night=<rgba 0.5 0.5 0.5 1>, Morning Threshold=4))
    ScaleA=false
    ScaleRGB=true
   ) {module=ScaleColor002}
  ]
  rendererClasses [&NiagaraMeshRendererProperties]
  renderers [{rendererClass=&NiagaraMeshRendererProperties, rendererIndex=0}]
 }
 Icy_Mist {
  emitterUpdateScript [
   &EmitterState(Life Cycle Mode=::NewEnumerator1, Loop Behavior=::NewEnumerator1, Loop Delay=0.6, Loop Duration Mode=::NewEnumerator1, UseLoopDelay=true)
   &SpawnRate(
    Spawn Probability=&MakeCustomFloatFromBool(Bool=$User.bIsFullyDeployed, True Float=1, False Float=0)
    SpawnRate=&Lerp_Float(A=10, B=20, Alpha=$System.Timeline.Time)
    Use Spawn Probability=true
   )
  ]
  particleSpawnScript [
   &PartitionParticles()
   &InitializeParticle(
    Color=&MergeHelper_ColorSwitch(
     0=&RandomRangeLinearColor(Minimum=<rgba 0.5972 0.7837 0.9323 0.5>, Maximum=<rgba 0.1892 0.2908 0.3906 0.5>)
     1=&RandomRangeLinearColor(Minimum=<rgba 0.9323 0.7927 0.5972 0.5>, Maximum=<rgba 0.3906 0.2228 0.1892 0.5>)
     2=?
     3=?
     Index=&MakeIntFromFloat(Float=$User.Variant)
     Num=::NewEnumerator0
    )
    Lifetime=&RandomRangeFloat(Minimum=1, Maximum=2)
    Sprite Rotation Mode=::NewEnumerator2
    Sprite Size Mode=::NewEnumerator3
    Sprite UV Mode=::NewEnumerator3
    Uniform Sprite Size=&RandomRangeFloat(Minimum=32, Maximum=50)
    UsePositionOffset=false
    Write Material Random=true
   )
   &SkeletalMeshLocation(
    Bone / Socket Index=$Particles.PartitionParticles.Partition
    Bone Sampling Mode=::NewEnumerator4
    Interpolate Mesh Sampling=false
    MaskSpawnGroup=false
    Skeletal Mesh=$System.Skeletal Mesh
   )
   &AddVelocity(Velocity=&FN_FrontEndVector3Modifier(
     Gameplay=&Multiply_Vector(
      A=&MergeHelper_VectorSwitch(
       0=&RandomRangeVector(Minimum=<-35 150 -35>, Maximum=<-35 250 125>)
       1=&RandomRangeVector(Minimum=<35 150 -35>, Maximum=<35 250 125>)
       2=?
       3=?
       Index=$Particles.PartitionParticles.Partition
       Num=::NewEnumerator0
      )
      B=&Lerp_Vector(A=<1 1 1>, B=<1 1.5 1>, Alpha=$System.Timeline.NormalizedTime)
     )
     Front End=&Multiply_Vector(
      A=&MergeHelper_VectorSwitch(
       0=&RandomRangeVector(Minimum=<-35 150 -35>, Maximum=<-35 250 125>)
       1=&RandomRangeVector(Minimum=<35 150 -35>, Maximum=<35 250 125>)
       2=?
       3=?
       Index=$Particles.PartitionParticles.Partition
       Num=::NewEnumerator0
      )
      B=&Lerp_Vector(A=<1 1 1>, B=<1 1.5 1>, Alpha=$System.Timeline.NormalizedTime)
     )
    ))
   &CameraOffset(Camera Offset Amount=5)
   &DynamicMaterialParameters(
    Index 0 Param 1=3
    Index 0 Param 2=1
    Index 0 Param 3=0.5
    Module.Index0Float0Write=true
    Module.Index0Float1Write=true
    Module.Index0Float2Write=true
    Module.Index0Float3Write=true
   )
  ]
  particleUpdateScript [
   &ParticleState()
   &Drag(Use Linear Drag=true, Use Rotational Drag=true)
   &SolveForcesAndVelocity(Clamp Velocity=false, Limit Acceleration=false)
   &ScaleSpriteSize(
    Uniform Curve Scale=2
    Uniform Curve Sprite Scale=DataInterface() {propertyValues {Curve {keys [<at 0 0.5 RCIM_Cubic>, {tangentMode=RCTM_User, time=1, value=1}]}}}
   )
   &SubUVAnimation(
    "Sprite Renderer "=DataInterface() {propertyValues {SpriteRenderer=&self:Icy_Mist_0.NiagaraSpriteRendererProperties_0}}
    SubUV Animation Mode:ENiagara_SubUVLookupModeV2=::NewEnumerator1
    SubUV Curve=DataInterface() {propertyValues {Curve {keys [
        {arriveTangent=1.5019, interpMode=RCIM_Cubic, leaveTangent=1.5019, tangentMode=RCTM_User, time=0, value=0}
        {arriveTangent=0.2086, interpMode=RCIM_Cubic, leaveTangent=0.2086, tangentMode=RCTM_User, time=1, value=1}
       ]}}}
    UseEndFrame=false
    UseStartFrame=false
   )
   &ScaleColor(
    Scale Alpha=&FloatFromCurve(
     FloatCurve=DataInterface() {propertyValues {Curve {keys [
         {interpMode=RCIM_Cubic, tangentMode=RCTM_User, time=0, value=0}
         {interpMode=RCIM_Cubic, tangentMode=RCTM_User, time=0.5, value=1}
         {tangentMode=RCTM_User, time=1, value=0}
        ]}}}
     Scale Curve=&FN_FrontEndFloatModifier(Gameplay=0.2, Front End=0.25)
    )
    ScaleA=true
    ScaleRGB=true
   )
   &KillParticles(Kill Particles=&Boolean_NotOperation(Bool=/Game/Effects/Niagara/DynamicInputs/State/FN_IsExecutionStateActive()))
  ]
  rendererClasses [&NiagaraSpriteRendererProperties]
  renderers [{rendererClass=&NiagaraSpriteRendererProperties}]
 }
}
summary {
 systemName=NS_Glider_BriskReign_Glacier_Trail
 emitters [
  {emitterName=Icy_Mist, rendererClasses [&NiagaraSpriteRendererProperties]}
  {emitterName=FN_AnimTrail_Dark, rendererClasses [&NiagaraRibbonRendererProperties]}
  {emitterName=FN_AnimTrail_Light, rendererClasses [&NiagaraRibbonRendererProperties]}
  {emitterName=Ice_Shards, rendererClasses [&NiagaraMeshRendererProperties]}
 ]
 userVariables [
  User.bIsFrontEnd {type=NiagaraBool}
  User.bIsFrontEndPreview {type=NiagaraBool}
  User.bIsFullyDeployed {type=NiagaraBool, defaultValue=true}
  User.ForwardDot {type=NiagaraFloat, defaultValue=0, description=""}
  User.FrontEndScale {type=NiagaraFloat, defaultValue=0, description=""}
  User.RibbonIDOffset {type=NiagaraInt32}
  User.RightDot {type=NiagaraFloat, defaultValue=0, description=""}
  User.Variant {type=NiagaraFloat, defaultValue=0, description=""}
 ]
}
